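; Fake fog light actor that imitates a glowing light volume with near and far sphere meshes and no runtime logic.
; Its construction script creates a dynamic material on each sphere, passing color, brightness, time-of-day multiplier and fade settings from variables.
; Both spheres derive light and sphere radii from the near sphere's smaller X/Y world scale, plus a margin for the sphere.
; Afterwards it forces uniform actor scale by averaging the three axes.
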(event EventBeginPlay)

(event Collision|EventActorBeginOverlap
  (params (OtherActor :type Actor%20Object%20Reference)))

(event EventTick
  (params (DeltaSeconds :type Float%20%28single-precision%29)))

(fn ConstructionScript
  (bind _sm_fakelightsphere (Variables|Default|GetSM_FakeLightSphere))
  (bind _returnvalue (Transformation|GetWorldScale _sm_fakelightsphere))
  (bind _x (.x _returnvalue))
  (bind _y (.y _returnvalue))
  (bind _x_2 (.x _returnvalue))
  (bind _y_2 (.y _returnvalue))
  (bind _returnvalue_2 9e+01)
  (bind _returnvalue_3 (Math|Float|Min(Float) _x _y))
  (bind _returnvalue_4 (* _returnvalue_2 _returnvalue_3))
  (bind _returnvalue_5 (Math|Float|Min(Float) _x_2 _y_2))
  (bind _returnvalue_6 (* _returnvalue_2 _returnvalue_5))
  (Utilities|FlowControl|Sequence
    (:then_1
      (bind _returnvalue_7 (Rendering|Material|CreateDynamicMaterialInstance _sm_fakelightsphere 0 0 "None" :!class /Script/Engine.PrimitiveComponent))
      (Rendering|Material|SetScalarParameterValue _returnvalue_7 "LightRadius" _returnvalue_6 :!class /Script/Engine.MaterialInstanceDynamic)
      (Rendering|Material|SetScalarParameterValue _returnvalue_7 "SphereRadius" (+ _returnvalue_6 (* (Variables|Default|GetSphereRadiusMargin) _returnvalue_5)) :!class /Script/Engine.MaterialInstanceDynamic)
      (Rendering|Material|SetScalarParameterValue _returnvalue_7 "Brightness" (Variables|Default|GetBrightness) :!class /Script/Engine.MaterialInstanceDynamic)
      (Rendering|Material|SetScalarParameterValue _returnvalue_7 "CameraFadeLength" (Variables|Default|GetDrawFadeLength) :!class /Script/Engine.MaterialInstanceDynamic)
      (Rendering|Material|SetScalarParameterValue _returnvalue_7 "CameraFadeOffset" (Variables|Default|GetDrawDistance) :!class /Script/Engine.MaterialInstanceDynamic)
      (Rendering|Material|SetVectorParameterValue _returnvalue_7 "Color" (Math|Conversions|ToLinearColor(Color) (Variables|Default|GetColor)) :!class /Script/Engine.MaterialInstanceDynamic)
      (Rendering|Material|SetScalarParameterValue _returnvalue_7 "LightDistance" 0.0 :!class /Script/Engine.MaterialInstanceDynamic)
      (Rendering|Material|SetScalarParameterValue _returnvalue_7 "MinLightTop" (Variables|Default|GetBrightnesShadows) :!class /Script/Engine.MaterialInstanceDynamic)
      (Rendering|Material|SetScalarParameterValue _returnvalue_7 "NearFadeDistance" (Variables|Default|GetNearFadeDistance) :!class /Script/Engine.MaterialInstanceDynamic)
      (Rendering|Material|SetVectorParameterValue _returnvalue_7 "Emissive multiplier over day phases" (Math|Conversions|ToLinearColor(Color) (Variables|Default|GetTimeOfDayBrightness)) :!class /Script/Engine.MaterialInstanceDynamic)
      (Rendering|Material|SetScalarParameterValue _returnvalue_7 "DepthFadeDistance" (Variables|Default|GetDepthFade) :!class /Script/Engine.MaterialInstanceDynamic))
    (:then_2
      (bind _returnvalue_8 (Rendering|Material|CreateDynamicMaterialInstance (Variables|Default|GetSM_FakeLightSphereFar) 0 0 "None" :!class /Script/Engine.PrimitiveComponent))
      (Rendering|Material|SetScalarParameterValue _returnvalue_8 "LightRadius" _returnvalue_4 :!class /Script/Engine.MaterialInstanceDynamic)
      (Rendering|Material|SetScalarParameterValue _returnvalue_8 "SphereRadius" (+ _returnvalue_4 (* (Variables|Default|GetSphereRadiusMargin) _returnvalue_3)) :!class /Script/Engine.MaterialInstanceDynamic)
      (Rendering|Material|SetScalarParameterValue _returnvalue_8 "Brightness" (Variables|Default|GetBrightness) :!class /Script/Engine.MaterialInstanceDynamic)
      (Rendering|Material|SetScalarParameterValue _returnvalue_8 "CameraFadeLength" (Variables|Default|GetDrawFadeLength) :!class /Script/Engine.MaterialInstanceDynamic)
      (Rendering|Material|SetScalarParameterValue _returnvalue_8 "CameraFadeOffset" (Variables|Default|GetDrawDistance) :!class /Script/Engine.MaterialInstanceDynamic)
      (Rendering|Material|SetVectorParameterValue _returnvalue_8 "Color" (Math|Conversions|ToLinearColor(Color) (Variables|Default|GetColor)) :!class /Script/Engine.MaterialInstanceDynamic)
      (Rendering|Material|SetScalarParameterValue _returnvalue_8 "LightDistance" 0.0 :!class /Script/Engine.MaterialInstanceDynamic)
      (Rendering|Material|SetScalarParameterValue _returnvalue_8 "MinLightTop" (Variables|Default|GetBrightnesShadows) :!class /Script/Engine.MaterialInstanceDynamic)
      (Rendering|Material|SetScalarParameterValue _returnvalue_8 "NearFadeDistance" (Variables|Default|GetNearFadeDistance) :!class /Script/Engine.MaterialInstanceDynamic)
      (Rendering|Material|SetVectorParameterValue _returnvalue_8 "Emissive multiplier over day phases" (Math|Conversions|ToLinearColor(Color) (Variables|Default|GetTimeOfDayBrightness)) :!class /Script/Engine.MaterialInstanceDynamic)
      (Rendering|Material|SetScalarParameterValue _returnvalue_8 "DepthFadeDistance" (Variables|Default|GetDepthFade) :!class /Script/Engine.MaterialInstanceDynamic))
    (:then_3
      (Transformation|SetActorScale3D (Math|Vector|MakeVector (/ (Math|Float|float+float (.z (.scale (Transformation|GetActorTransform))) (.x (.scale (Transformation|GetActorTransform))) (.y (.scale (Transformation|GetActorTransform)))) 3.0) (/ (Math|Float|float+float (.z (.scale (Transformation|GetActorTransform))) (.x (.scale (Transformation|GetActorTransform))) (.y (.scale (Transformation|GetActorTransform)))) 3.0) (/ (Math|Float|float+float (.z (.scale (Transformation|GetActorTransform))) (.x (.scale (Transformation|GetActorTransform))) (.y (.scale (Transformation|GetActorTransform)))) 3.0))))))
