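# Translucent, unlit, two-sided material for glowing blue line effects in the Reality Daisy environment, likely around its giant tree.
# SplineThicken widens the strips, with switchable wind offset added from per-axis sines varied by vertex color and weighted along the strip.
# World-position-driven macro gusts blend in via vertex blue.
# An HDR blue tint scales with vertex green, while UV gradients and a panned mask texture shape opacity.
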Material /RealitySeedGameplay/Environment/Assets_Daisy/FX/M_Reality_Daisy_BlueLines
graph {
 properties {blendMode=BLEND_Translucent, pixelDepthOffsetMode=PDOM_Legacy, shadingModel=MSM_Unlit, twoSided=true, usedShadingModels=Unlit}
 nodes [
  Add(
   A=MaterialFunctionCall(
    WidthBase=ScalarParameter() {defaultValue=25, parameterName="WidthBase (S)"}
    WidthTip=ScalarParameter() {defaultValue=25, parameterName="WidthTip (S)"}
   ) {materialFunction=&SplineThicken}.WorldPositionOffset
   B=NamedRerouteUsage()
  )
  n1 = Time()
  n2 = ScalarParameter() {defaultValue=512, parameterName=MacroWindSpeed}
  n3 = Divide(
   A=Add(A=WorldPosition().XYZ, B=Multiply(A=@n1, B=@n2))
   B=VectorParameter() {parameterName=MacroWindScale, defaultValue=<rgba 2048 2048 2048 1>}.RGB
  )
  ChannelMaskParameter(None=@n3) {parameterName=MacroWindDirection}
  CollectionParameter() {collection=&FortniteMaterialParameters, parameterName=WindStrength}
  CollectionParameter() {collection=&FortniteMaterialParameters, parameterName=WindSpeed}
  n4 = VertexColor()
  Multiply(A=@n4.G, B=VectorParameter() {parameterName=Color, defaultValue=<rgba 1 1 25 1>}.RGB)
  n5 = TextureCoordinate()
  n6 = MaterialFunctionCall(UV Channel=@n5) {materialFunction=&LinearGradient}
  n7 = OneMinus(None=@n6.UGradient)
  n8 = TextureSampleParameter2D(UVs=Panner(Coordinate=TextureCoordinate() {coordinateIndex=2})) {
   parameterName=FXMask
   samplerType=SAMPLERTYPE_Masks
   texture=/Game/Environments/Artemis/Sets/Daisy/GiantTree_A/Textures/T_Daisy_GiantTree_Leaves01_MSK
  }
  Multiply(
   A=Multiply(A=Multiply(A=@n6.UGradient, B=@n7), B=ComponentMask(None=@n5) {g=true})
   B=ChannelMaskParameter(None=@n8.RGB) {parameterName=OpacityChannel}.R
  )
  n9 = Time()
  n10 = ScalarParameter() {parameterName=Min_TimeOffset}
  n11 = ScalarParameter() {defaultValue=1, parameterName=Max_TimeOffset}
  n12 = VertexColor()
  n13 = Add(A=@n9, B=LinearInterpolate(A=@n10, B=@n11, Alpha=@n12.R))
  n14 = Multiply(
   A=CollectionParameter() {collection=&FortniteMaterialParameters, parameterName=WindSpeed}
   B=LinearInterpolate(
    A=ScalarParameter() {defaultValue=0.25, parameterName=Min_SpeedOffset}
    B=ScalarParameter() {defaultValue=0.5, parameterName=Max_SpeedOffset}
    Alpha=@n12.G
   )
  )
  n15 = ScalarParameter() {defaultValue=0.5, parameterName=Min_StrengthOffset}
  n16 = ScalarParameter() {defaultValue=1, parameterName=Max_StrengthOffset}
  n17 = ChannelMaskParameter(None=OneMinus(None=TextureCoordinate())) {maskChannel=Green, parameterName=StripGradientChannel, defaultValue=<rgba 0 1 0 0>}
  n18 = Multiply(
   A=Multiply(A=ScalarParameter() {defaultValue=32, parameterName=Debug_WindStrength}, B=LinearInterpolate(A=@n15, B=@n16, Alpha=@n12.R))
   B=@n17.R
  )
  n19 = Multiply(A=Sine(None=Multiply(A=@n13, B=@n14)), B=@n18)
  n20 = Sine(None=@n3)
  n21 = ScalarParameter() {defaultValue=512, parameterName=WindStrength}
  n22 = MaterialFunctionCall(Float2=Multiply(A=ConstantBiasScale(None=@n20), B=Multiply(A=@n17.R, B=@n21))) {materialFunction=&BreakOutFloat2Components}
  n23 = MaterialFunctionCall(X=@n22.R, Y=@n22.R, Z=Constant()) {materialFunction=&MakeFloat3}
  n24 = LinearInterpolate(A=@n19, B=Add(A=@n19, B=@n23.Result), Alpha=@n12.B)
  n25 = Add(A=@n9, B=LinearInterpolate(A=@n10, B=@n11, Alpha=@n12.G))
  n26 = Multiply(A=Sine(None=Multiply(A=@n25, B=@n14)), B=@n18)
  n27 = LinearInterpolate(A=@n26, B=Add(A=@n26, B=@n23.Result), Alpha=@n12.B)
  n28 = Add(A=@n9, B=LinearInterpolate(A=@n10, B=@n11, Alpha=@n12.G))
  n29 = Multiply(A=Sine(None=Multiply(A=@n28, B=@n14)), B=@n18)
  NamedRerouteDeclaration(None=StaticSwitchParameter(
    True=MaterialFunctionCall(X=@n24, Y=@n27, Z=@n29) {materialFunction=&MakeFloat3}.Result
    False=Constant()
   ) {parameterName=UseWindAnimation}) {name=WindWPO, nodeColor=<rgba 1 0.4219 0 1>}
  OneMinus(None=@n4.G)
  PerInstanceRandom()
  VectorParameter() {parameterName=MAcroWindSpeed, defaultValue=<rgba 512 512 512 1>}
 ]
}
