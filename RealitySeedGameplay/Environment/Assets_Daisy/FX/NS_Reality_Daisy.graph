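# NS_Reality_Daisy is a Niagara environment effect that fills a large spherical area with drifting multicolored sprites.
# A single GPU emitter spawns continuously, tinting each particle from a random color-curve sample with random channel boosts.
# Upward gravity and panning curl noise move particles, while a region mask makes them brighter near the center.
# Sprites fade and pulse in size, scale up with camera distance via scalability distance fraction, and use distance culling.
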NiagaraSystem /RealitySeedGameplay/Environment/Assets_Daisy/FX/NS_Reality_Daisy
dependencies {
 usedDataInterfaces [&NiagaraDataInterfaceColorCurve, &NiagaraDataInterfaceDebugDraw, &NiagaraDataInterfaceCurve]
 usedDynamicInputs []
 usedModules [&SystemState]
}
emitters {Empty {
  simTarget=GPUComputeSim
  emitterUpdateScript [
   &EmitterState(
    Enable Distance Culling=true
    MaxDistance=35000
    Module.bUseMaxDistance=true
    Module.bUseMinDistance=false
    Scalability Mode=::NewEnumerator1
   )
   &SpawnRate(SpawnRate=150, Use Spawn Probability=false)
  ]
  particleSpawnScript [
   &InitializeParticle(
    Color=&MultiplyLinearColor(
     Linear Color=&ColorFromCurve(
      ColorCurve=DataInterface() {propertyValues {
        AlphaCurve {keys [<at 0 1 RCIM_Cubic>]}
        BlueCurve {keys [<at 0 0 RCIM_Cubic>,<at 0.408 1 RCIM_Cubic>,<at 0.9128 0>,<at 1 1 RCIM_Cubic>]}
        GreenCurve {keys [
          {arriveTangent=0.654, interpMode=RCIM_Cubic, time=0, value=0.3232}
          {arriveTangent=0.7415, interpMode=RCIM_Cubic, leaveTangent=0.7415, time=0.408, value=0.6084}
          <at 0.9128 1>
          <at 1 1 RCIM_Cubic>
         ]}
        RedCurve {keys [
          {arriveTangent=1.67, interpMode=RCIM_Cubic, time=0, value=1}
          {arriveTangent=-1.0512, interpMode=RCIM_Cubic, leaveTangent=-1.0512, time=0.408, value=0}
          <at 0.9128 0.0404>
          <at 1 1 RCIM_Cubic>
         ]}
       }}
      CurveIndex=&RandomRangeFloat(Minimum=0, Maximum=1)
     )
     Scale Factor=&MakeVector4(
      X=&RandomRangeFloat(Minimum=0, Maximum=5)
      Y=&RandomRangeFloat(Minimum=1, Maximum=3)
      Z=&RandomRangeFloat(Minimum=1, Maximum=2)
      W=1
     )
    )
    Lifetime Max=3
    Lifetime Min=2
    Lifetime Mode=::NewEnumerator1
    Sprite Size Mode=::NewEnumerator1
    Sprite UV Mode=::NewEnumerator1
    Uniform Sprite Size Max=500
    Uniform Sprite Size Min=100
    UsePositionOffset=false
    Write Material Random=true
   )
   &ShapeLocation(Hemisphere Distribution=<1 0.5>, MaskSpawnGroup=false, Override Seed=false, Sphere Radius=8000, UseOwnerScale=true)
  ]
  particleUpdateScript [
   &ParticleState()
   &CurlNoiseForce(Mask Curl Noise=true, Noise Strength=2500, Pan Noise=true, Pan Noise Field=<0 0 25>)
   &GravityForce(Gravity=<0 0 500>)
   &SolveForcesAndVelocity(Clamp Velocity=true, Limit Acceleration=false)
   &RegionMask(Inner Float=3, Outer Float=0.5, Region Falloff=1, Region Radius=8000, UseOffset=false, UseRotation=false)
   &ScaleColor(
    Scale Alpha=&FloatFromCurve(FloatCurve=DataInterface() {propertyValues {Curve {keys [
         {arriveTangent=9.0725, interpMode=RCIM_Cubic, leaveTangent=9.0725, tangentMode=RCTM_User, time=0, value=0}
         {arriveTangent=-2.5249, interpMode=RCIM_Cubic, leaveTangent=-2.5249, tangentMode=RCTM_User, time=0.7, value=0.85}
         {arriveTangent=-0.0405, interpMode=RCIM_Cubic, leaveTangent=-0.0405, tangentMode=RCTM_User, time=1, value=0}
        ]}}})
    Scale RGB=&Multiply_VectorByFloat(Vector=<1 1 1>, Float=$Particles.RegionMask.Float)
    ScaleA=true
    ScaleRGB=true
   )
   &ScaleSpriteSize(
    Uniform Curve Scale=&Multiply_Float(
     A=&Waveform(
      Global Amplitude Scale=&RandomRangeFloat(Minimum=0.5, Maximum=1)
      "Waveform [1]"=::NewEnumerator0
      "[1] Phase Scale"=&RandomRangeFloat(Minimum=0.25, Maximum=0.5)
      "[1] Frequency"=1
      "[1] Amplitude Min/Max"=<0 1>
     )
     B=/Niagara/DynamicInputs/Scalability/ScalabilityDistanceBasedFloat(
      Curve:NiagaraDataInterfaceCurve=DataInterface() {propertyValues {Curve {keys [<at 0 0.25>,<at 1 5>]}}}
      CurveIndex:NiagaraFloat=$Emitter.Scalability.DistanceFraction
      Bias:NiagaraFloat=0
      Scale:NiagaraFloat=1
     )
    )
    Uniform Curve Sprite Scale=DataInterface() {propertyValues {Curve {keys [
        {arriveTangent=9.0725, interpMode=RCIM_Cubic, leaveTangent=9.0725, tangentMode=RCTM_User, time=0, value=0}
        {arriveTangent=-2.5249, interpMode=RCIM_Cubic, leaveTangent=-2.5249, tangentMode=RCTM_User, time=0.1408, value=0.9847}
        {arriveTangent=-0.0405, interpMode=RCIM_Cubic, leaveTangent=-0.0405, tangentMode=RCTM_User, time=1, value=0}
       ]}}}
   )
  ]
  rendererClasses [&NiagaraSpriteRendererProperties]
  renderers [{rendererClass=&NiagaraSpriteRendererProperties}]
 }}
summary {
 systemName=NS_Reality_Daisy
 emitters [{emitterName=Empty, simTarget=GPUComputeSim, rendererClasses [&NiagaraSpriteRendererProperties]}]
 userVariables []
}
